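; Gameplay ability that clears the jump boost FX, going by its name, when the player lands.
; On activation it casts the avatar to FortPlayerPawn and stores it, ending immediately if the cast fails.
; A Sequence starts movement-mode waits for Walking and Custom, the latter a fix for FX lingering after water landings.
; The ability cleans up by removing the effect that granted it and ending itself.
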(event Ability|EventActivateAbility
  (bind _asfort_player_pawn (Utilities|Casting|CastToFortPlayerPawn (Ability|GetAvatarActorfromActorInfo))
    (:then
      (Variables|Default|SetPlayerPawn _asfort_player_pawn)
      (Utilities|FlowControl|Sequence
        (:then_0
          (Ability|Tasks|WaitMovementModeChange "MOVE_Walking"
            (:OnChange
              _)))
        (:then_1
          (Ability|Tasks|WaitMovementModeChange "MOVE_Custom"
            (:OnChange
              _)) ; Fix for fx not getting removed when landing in water (FORT-223198)
        ))
      (Ability|RemoveGrantedbyEffect)
      (Ability|EndAbility))
    (:CastFailed
      (Ability|EndAbility))))

(event Ability|EventFailedToActivatePassiveAbility
  (params (ActorInfo :type Gameplay%20Ability%20Actor%20Info%20Structure))
  (|CallParentFunction))
